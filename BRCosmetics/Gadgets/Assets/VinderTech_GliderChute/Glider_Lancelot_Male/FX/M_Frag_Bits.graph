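# Lit material for Niagara mesh-particle fragment bits in the Lancelot glider FX, mostly reusing Lancelot skydiving contrail textures.
# A camera-to-normal dot product samples a gold curve-atlas row, which tints the diffuse texture inside a mask.
# That color is summed with a particle-colored FX mask and, scaled down, feeds an emissive adjustment function.
# A normalized normal map and OARM channels, adjusted by bias and a metal-mask power, supply surface detail.
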Material /BRCosmetics/Gadgets/Assets/VinderTech_GliderChute/Glider_Lancelot_Male/FX/M_Frag_Bits
graph {
 properties {bUsedWithNiagaraMeshParticles=true, pixelDepthOffsetMode=PDOM_Legacy, usedShadingModels=Default Lit}
 nodes [
  n1 = CameraVectorWS()
  n2 = PixelNormalWS()
  n3 = CurveAtlasRowParameter(CurveTime=Clamp(None=DotProduct(A=@n1, B=@n2))) {atlas=&AthenaCurveAtlas01, curve=/Game/Athena/LinearColorCurves/Curve_IcoQueenColor_Gold, defaultValue=55}
  n4 = LinearInterpolate(
   B=Multiply(
    A=ChannelMaskParameter(None=@n3) {group=LUT, parameterName=LUTChannel}.R
    B=TextureSample() {texture=/BRCosmetics/Athena/Effects/Particles/Skydiving/NS_Contrail_Lancelot/T_Frag_Diffuse}.RGB
   )
   Alpha=TextureSample() {texture=/BRCosmetics/Athena/Effects/Particles/Skydiving/NS_Contrail_Lancelot/LP_Frag_lambert1_MaskMap}.G
  ) {constAlpha=0}
  Add(
   A=Multiply(
    A=TextureSample() {samplerType=SAMPLERTYPE_Masks, texture=&/T_Frag_FX_Mask}.RGB
    B=ParticleColor().RGB
   )
   B=@n4
  )
  n5 = Multiply(A=@n4) {constB=0.1}
  MaterialFunctionCall(Emissive=@n5) {materialFunction=&MF_FX_EmissiveAdjustments}
  Reroute(None=@n5)
  Reroute(None=Normalize(VectorInput=TextureSample() {samplerType=SAMPLERTYPE_Normal, texture=/BRCosmetics/Athena/Effects/Particles/Skydiving/NS_Contrail_Lancelot/T_Frag_Normal}.RGB))
  n6 = TextureSample() {texture=/BRCosmetics/Athena/Effects/Particles/Skydiving/NS_Contrail_Lancelot/T_Frag_OARM}
  n7 = Multiply(A=@n6.G)
  Reroute(None=Saturate(None=Add(A=@n7) {constB=0.2}))
  n8 = ScalarParameter() {defaultValue=0.5, group=Masks, parameterName=MetalMaskPower}
  Reroute(None=Saturate(None=Power(Base=@n6.B, Exp=@n8)))
  TextureSample() {texture=&/T_Frag_Diffuse_Glider}
  TextureSample() {texture=/BRCosmetics/Athena/Effects/Particles/Skydiving/NS_Contrail_Lancelot/LP_Frag_lambert1_Specular}
 ]
}
